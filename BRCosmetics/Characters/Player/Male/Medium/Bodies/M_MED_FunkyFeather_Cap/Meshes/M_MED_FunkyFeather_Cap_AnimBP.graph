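# Animation Blueprint for the M_MED_FunkyFeather_Cap cosmetic body mesh, following its attached parent and correcting head and neck rotation.
# The AnimGraph feeds the copied parent pose through a Control Rig, then bone-space Modify Bone nodes on head, neck_01 and neck_02.
# Each update negates and scales the head's ref-pose delta against neck_02, remapping yaw to vary head yaw scaling.
# Neck yaw's multiply lacks its yaw input, unlike the head's.
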AnimBlueprint /BRCosmetics/Characters/Player/Male/Medium/Bodies/M_MED_FunkyFeather_Cap/Meshes/M_MED_FunkyFeather_Cap_AnimBP
target_skeleton=&/M_MED_FunkyFeather_Cap_Skeleton
graphs [
 AnimGraph {nodes [
   n1 = Control Rig(Source=Copy Pose From Mesh() {bUseAttachedParent=true}) {controlRigAssetReference {blueprintRigClass=&/M_MED_FunkyFeather_Cap_CtrlRig.M_MED_FunkyFeather_Cap_CtrlRig_C}}
   n2 = "Transform (Modify) Bone - Bone: head"!(ComponentPose=Local To Component(LocalPose=@n1), Alpha=0.5, Rotation=Get Head Rot() -> (Head Rot:Rotator)) -> (Pose) {
    rotationMode=BMM_Replace
    rotationSpace=BCS_BoneSpace
    translationSpace=BCS_BoneSpace
    alphaBoolBlend {blendInTime=0.1, blendOption=Linear, blendOutTime=0.1}
    boneToModify {boneName=head}
   }
   n3 = Get Neck Rot() -> (Neck Rot:Rotator)
   n4 = "Transform (Modify) Bone - Bone: neck_01"!(ComponentPose=@n2, Alpha=0.8, Rotation=@n3) -> (Pose) {
    rotationMode=BMM_Replace
    rotationSpace=BCS_BoneSpace
    translationSpace=BCS_BoneSpace
    alphaBoolBlend {blendInTime=0.1, blendOption=Linear, blendOutTime=0.1}
    boneToModify {boneName=neck_01}
   }
   n5 = "Transform (Modify) Bone - Bone: neck_02"!(ComponentPose=@n4, Alpha=0.8, Rotation=@n3) -> (Pose) {
    rotationMode=BMM_Replace
    rotationSpace=BCS_BoneSpace
    translationSpace=BCS_BoneSpace
    alphaBoolBlend {blendInTime=0.1, blendOption=Linear, blendOutTime=0.1}
    boneToModify {boneName=neck_02}
   }
   Output Pose(Result=Component To Local(ComponentPose=@n5)) {layerGroup=None}
  ]}
 EventGraph {nodes [
   Event BlueprintUpdateAnimation()
   n1 = BreakRotator(InRot=GetDeltaTransformFromRefPose!(self=GetOwnerSkeletalMeshComponent(), BoneName=head, BaseName=neck_02) -> (ReturnValue_Location:Vector, ReturnValue_Rotation:Rotator, ReturnValue_Scale:Vector).ReturnValue_Rotation)
   n2 = Remap Float(Input=@n1.Yaw, OldRange=(X=10,Y=-28), NewRange=(X=0,Y=1), Clamp=true) -> (Output=~)
   n3 = Lerp(A=Get Z Head Rot Per Min() -> (Z Head Rot Per Min:double), B=Get Z Head Rot Per Max() -> (Z Head Rot Per Max:double), Alpha=@n2)
   n4 = float * float!(A=@n1, B=-1, C:double=Get X Head Rot Per() -> (X Head Rot Per:double)) -> (ReturnValue) {is=title K2Node_PromotableOperator}
   n5 = float * float(A=@n1.Pitch, B=-1, C=Get Y Head Rot Per() -> (Y Head Rot Per:double))
   > Set Head Rot(Head Rot:Rotator=MakeRotator(Roll=@n4, Pitch=@n5, Yaw=float * float(A=@n1.Yaw, B=-1, C=@n3))) -> (Output_Get:Rotator)
   n6 = Lerp(A=Get Z Head Rot Per Min(), B=0.5, Alpha=@n2)
   > Set Neck Rot(Neck Rot:Rotator=MakeRotator(Roll=@n4, Pitch=@n5, Yaw=float * float(B=-1, C=@n6))) -> (Output_Get:Rotator)
  ]}
]
variables [
 Head Rot {type=Rotator}
 Neck Rot {type=Rotator}
 X Head Rot Per {type=number, default=0.4}
 Y Head Rot Per {type=number, default=0.3}
 Z Head Rot Per Max {type=number, default=0.5}
 Z Head Rot Per Min {type=number}
]
