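# Physics asset for the face accessory mesh of the M_MED_FunkyFeather_Cap male body, covering a short dynamic chain on the upper torso.
# Capsule bodies cover spine_05 and both clavicles as torso volumes, alongside dyn_Chain_base, dyn_Chain_01 and dyn_Chain_02.
# Constraints link the chain in series up to spine_05, with narrow swing and twist limits between chain links.
# The chain base and both clavicles attach to spine_05 with free angular motion.
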PhysicsAsset /BRCosmetics/Characters/Player/Male/Medium/Bodies/M_MED_FunkyFeather_Cap/Meshes/Parts/M_MED_FunkyFeather_Cap_FaceAcc_Physics
bodies {
 clavicle_l [{shapeType=Capsule, length=7, radius=8, shapeName=clavicle_l_capsule, center=<8.8029 0 0>, rotation=<rot -90 0 0>}]
 clavicle_r [{shapeType=Capsule, length=7, radius=8, shapeName=clavicle_r_capsule, center=<-5.0813 0 0>, rotation=<rot 90 0 0>}]
 dyn_Chain_01 [{shapeType=Capsule, length=7.191, radius=3.1156, shapeName=dyn_Chain_01_capsule, center=<3.9444 0 -2.5683>, rotation=<rot 0 0 90>}]
 dyn_Chain_02 [{shapeType=Capsule, length=4.749, radius=3.5204, shapeName=dyn_Chain_02_capsule, center=<5.2699 0 -1.8744>, rotation=<rot 0 0 90>}]
 dyn_Chain_base [{shapeType=Capsule, length=0.505, radius=0.505, shapeName=dyn_Chain_base_capsule, rotation=<rot 0 -90 0>}]
 spine_05 [{shapeType=Capsule, length=18, radius=9.5, shapeName=spine_05_capsule}]
}
constraints [
 {bone1Name=dyn_Chain_01, bone2Name=dyn_Chain_base, swing1LimitDegrees=20, swing2LimitDegrees=30, twistLimitDegrees=10}
 {bone1Name=dyn_Chain_02, bone2Name=dyn_Chain_01, swing1LimitDegrees=20, swing2LimitDegrees=40, twistLimitDegrees=10}
 {
  bone1Name=dyn_Chain_base
  bone2Name=spine_05
  swing1LimitDegrees=45
  swing1Motion=Free
  swing2LimitDegrees=45
  swing2Motion=Free
  twistLimitDegrees=45
  twistMotion=Free
 }
 {
  bone1Name=clavicle_l
  bone2Name=spine_05
  swing1LimitDegrees=45
  swing1Motion=Free
  swing2LimitDegrees=45
  swing2Motion=Free
  twistLimitDegrees=45
  twistMotion=Free
 }
 {
  bone1Name=clavicle_r
  bone2Name=spine_05
  swing1LimitDegrees=45
  swing1Motion=Free
  swing2LimitDegrees=45
  swing2Motion=Free
  twistLimitDegrees=45
  twistMotion=Free
 }
]
